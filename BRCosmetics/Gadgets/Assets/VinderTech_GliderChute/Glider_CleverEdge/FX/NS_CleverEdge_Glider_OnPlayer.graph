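# Niagara effect for the CleverEdge glider that plays short-lived dust sprites on the player's skeletal mesh.
# Two delayed, slightly staggered bursts use spawn groups to pick the bone or socket, mirroring offset and facing between them.
# Location and facing update every frame from the sampled mesh, keeping sprites attached as they animate.
# Two sprite renderers draw particles tinted by time of day, with a lifetime curve driving a dynamic material parameter.
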NiagaraSystem /BRCosmetics/Gadgets/Assets/VinderTech_GliderChute/Glider_CleverEdge/FX/NS_CleverEdge_Glider_OnPlayer
dependencies {
 usedDataInterfaces [&NiagaraDataInterfaceCurve, &NiagaraDataInterfaceSkeletalMesh]
 usedDynamicInputs []
 usedModules [
  &self:SystemSpawnScript.SystemScriptSource.SystemScriptGraph.NiagaraNodeAssignment_0.SetVariables_#1
  &SystemState
 ]
}
emitters {Dust {
  emitterUpdateScript [
   &EmitterState(Life Cycle Mode=::NewEnumerator1, Loop Behavior=::NewEnumerator1, Loop Delay=2.1, Loop Duration=1, UseLoopDelay=true)
   &SpawnBurst_Instantaneous(Use Spawn Probability=false, UseLoopCountLimit=false)
   &SpawnBurst_Instantaneous(Spawn Group=1, Spawn Time=0.025, Use Spawn Probability=false, UseLoopCountLimit=false) {module=SpawnBurst_Instantaneous001}
  ]
  particleSpawnScript [
   &InitializeParticle(
    Color=&FNTimeOfDayColor(Night Color=<rgba 1 1 1 0.1>, Morning Color=<rgba 1 1 1 0.35>, Day Color=<rgba 1 1 1 0.35>, Evening Color=<rgba 1 1 1 0.35>)
    Lifetime Max=0.25
    Lifetime Min=0.3
    Lifetime Mode=::NewEnumerator1
    Mass Max=2
    Mass Mode=::NewEnumerator3
    Sprite Rotation Angle=&RandomRangeFloat(Minimum=0, Maximum=1)
    Sprite Rotation Mode=::NewEnumerator1
    Sprite Size Mode=::NewEnumerator3
    Uniform Sprite Size=60
    UsePositionOffset=false
    Write Material Random=true
   )
   &CameraOffset(Camera Offset Amount=10)
   /SetVariables(Particles.SpriteRot2:NiagaraFloat=&RandomRangeFloat(Minimum=0, Maximum=1)) {isSetParameters=true}
  ]
  particleUpdateScript [
   &ParticleState()
   &ScaleColor(
    Scale Alpha=&FloatFromCurve(FloatCurve=DataInterface() {propertyValues {Curve {keys [<at -0.0046 1.0113 RCIM_Cubic>,<at 1.0011 1 RCIM_Cubic>]}}})
    Scale RGB=&VectorFromFloat(Value=3.3333)
    ScaleA=true
    ScaleRGB=true
   )
   &SkeletalMeshLocation(
    Bone / Socket Index=&SpawnGroup()
    Bone Sampling Mode=::NewEnumerator4
    MaskSpawnGroup=false
    Sampled Position Offset=&MaskVectorBySpawnGroup(Masked Vector=<-7 0 0>, Passthrough Vector=<7 0 0>)
    Skeletal Mesh=$System.Skeletal Mesh
   )
   &SolveForcesAndVelocity(Clamp Velocity=false, Limit Acceleration=false)
   &DynamicMaterialParameters(
    Index 0 Param 1=&FloatFromCurve(FloatCurve=DataInterface() {propertyValues {Curve {keys [
         {arriveTangentWeight=0.7991, leaveTangentWeight=0.7991, time=0, value=0}
         {arriveTangent=1.9332, interpMode=RCIM_Cubic, leaveTangent=1.9332, tangentMode=RCTM_User, time=1, value=1}
        ]}}})
    Index 0 Param 2=1
    Module.Param0WriteEnabled=true
    Module.Param1WriteEnabled=true
    Module.Param2WriteEnabled=true
    Module.Param3WriteEnabled=true
   )
   &SpriteFacingAndAlignment(
    Sprite Alignment=&RandomVector(Vector Scale=1)
    Sprite Facing=&Multiply_VectorByFloat(
     Vector=&ReturnMeshOrientationAxis(Mesh Orientation Axis=::XAxis, Mesh Orientation Quat=$Particles.SkeletalMeshLocation.SampledRotation)
     Float=&MaskFloatBySpawnGroup(Masked Float=1, Passthrough Float=-1)
    )
    WriteAlignment=true
    WriteFacing=true
   )
  ]
  rendererClasses [&NiagaraSpriteRendererProperties]
  renderers [
   {rendererClass=&NiagaraSpriteRendererProperties}
   {rendererClass=&NiagaraSpriteRendererProperties, rendererIndex=1}
  ]
 }}
summary {
 systemName=NS_CleverEdge_Glider_OnPlayer
 emitters [{emitterName=Dust, rendererClasses [&NiagaraSpriteRendererProperties]}]
 userVariables []
}
